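# Animation Blueprint for the male medium Order padded armor body, following the attached parent mesh's pose and adding secondary motion.
# The copied pose passes through a deform Control Rig, then an armor Control Rig fed the crouching state.
# Weighted constraints then position knees, thigh rings and torso pouches, with spring-based AnimDynamics on rings and pouches; only the left ring has a planar limit.
# The hood-to-head constraint has zero alpha; the metal-box bone node is unconnected.
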AnimBlueprint /Game/Characters/Player/Male/Medium/Bodies/M_Med_Padded_Armor_Order/Meshes/M_MED_Padded_Armor_Order_AnimBP
target_skeleton=&/M_MED_Padded_Armor_Order_Skeleton
graphs [
 AnimGraph {nodes [
   "Transform (Modify) Bone - Bone: dyn_metalbox"(Alpha=1, Translation="0,0,0", Rotation="0,0,0", Scale="1,1,1") {rotationMode=BMM_Additive, translationMode=BMM_Additive, alphaBoolBlend {blendOption=Linear}, boneToModify {boneName=dyn_metalbox}}
   n1 = Control Rig(Source=Copy Pose From Mesh() {bUseAttachedParent=true}) {
    alphaBoolBlend {blendOption=Linear}
    controlRigAssetReference {blueprintRigClass=&/F_MED_DeformRig_PaddedArmor_Order_CtrlRig.F_MED_DeformRig_PaddedArmor_Order_CtrlRig_C}
   }
   n2 = Control Rig!(Source=@n1, Alpha, CrouchingCtrl:bool=Get bIsCrouching()) -> (Pose) {
    alphaBoolBlend {blendOption=Linear}
    controlRigAssetReference {blueprintRigClass=&/M_MED_Padded_Armor_Order_CtrlRig.M_MED_Padded_Armor_Order_CtrlRig_C}
   }
   n3 = "Constraint - knee_l (PP)"!(ComponentPose=Local To Component(LocalPose=@n2), Alpha=1, ConstraintWeights_0=0.6, ConstraintWeights_1:float=0.4) -> (Pose) {
    alphaBoolBlend {blendOption=Linear}
    boneToModify {boneName=knee_l}
    constraintSetup [{targetBone {boneName=thigh_l}}, {targetBone {boneName=calf_l}}]
    constraintWeights [1,1]
   }
   n4 = "Constraint - knee_r (PP)"!(ComponentPose=@n3, Alpha=1, ConstraintWeights_0=0.6, ConstraintWeights_1:float=0.4) -> (Pose) {
    alphaBoolBlend {blendOption=Linear}
    boneToModify {boneName=knee_r}
    constraintSetup [{targetBone {boneName=thigh_r}}, {targetBone {boneName=calf_r}}]
    constraintWeights [1,1]
   }
   n5 = "Constraint - thighRing_l_zero (PPP)"!(ComponentPose=@n4, Alpha=1, ConstraintWeights_0=0.58, ConstraintWeights_1:float=0.08, ConstraintWeights_2:float=0.34) -> (Pose) {
    alphaBoolBlend {blendOption=Linear}
    boneToModify {boneName=thighRing_l_zero}
    constraintSetup [{targetBone {boneName=pelvis}}, {targetBone {boneName=spine_01}}, {targetBone {boneName=thigh_twist_01_l}}]
    constraintWeights [1,1,1]
   }
   n6 = "Constraint - thighRing_r_zero (PPP)"!(ComponentPose=@n5, Alpha=1, ConstraintWeights_0=0.58, ConstraintWeights_1:float=0.08, ConstraintWeights_2:float=0.34) -> (Pose) {
    alphaBoolBlend {blendOption=Linear}
    boneToModify {boneName=thighRing_r_zero}
    constraintSetup [{targetBone {boneName=pelvis}}, {targetBone {boneName=spine_01}}, {targetBone {boneName=thigh_twist_01_r}}]
    constraintWeights [1,1,1]
   }
   n7 = Anim Dynamics - Bone: thighRing_l(ComponentPose=@n6, Alpha=1, ExternalForce="0,0,0") {
    angularDampingOverride=1
    angularSpringConstant=500
    bAngularSpring=true
    bGravityOverrideInSimSpace=true
    bOverrideAngularDamping=true
    gravityScale=0.35
    simulationSpace=RootRelative
    alphaBoolBlend {blendOption=Linear}
    boundBone {boneName=thighRing_l}
    physicsBodyDefinitions [{
      sphereCollisionRadius=0
      boundBone {boneName=thighRing_l}
      boxExtents=<10 20 20>
      constraintSetup {angularLimitsMax=<0 0 10>, angularLimitsMin=<0 -50 -10>, angularTarget=<0 0 0>}
      localJointOffset=<10 0 0>
     }]
    planarLimits [{drivingBone {boneName=thigh_l}, planeTransform {location=<0 0 0>, rotation=<rot 20 0 130>, scale=<0.3 0.3 0.3>}}]
   }
   n8 = Anim Dynamics - Bone: thighRing_r(ComponentPose=@n7, Alpha=1, ExternalForce="0,0,0") {
    angularDampingOverride=1
    angularSpringConstant=500
    bAngularSpring=true
    bGravityOverrideInSimSpace=true
    bOverrideAngularDamping=true
    gravityScale=0.35
    simulationSpace=RootRelative
    alphaBoolBlend {blendOption=Linear}
    boundBone {boneName=thighRing_r}
    physicsBodyDefinitions [{
      sphereCollisionRadius=0
      boundBone {boneName=thighRing_r}
      boxExtents=<10 20 20>
      constraintSetup {angularLimitsMax=<0 0 10>, angularLimitsMin=<0 -50 -10>, angularTarget=<0 0 0>}
      localJointOffset=<10 0 0>
     }]
   }
   n9 = "Constraint - torsoPouch_1_l_zero (PPPP)"!(
    ComponentPose=@n8
    Alpha=1
    ConstraintWeights_0=0.1
    ConstraintWeights_1:float=0.24
    ConstraintWeights_2:float=0.48
    ConstraintWeights_3:float=0.18
   ) -> (Pose) {
    alphaBoolBlend {blendOption=Linear}
    boneToModify {boneName=torsoPouch_1_l_zero}
    constraintSetup [{targetBone {boneName=spine_01}}, {targetBone {boneName=spine_02}}, {targetBone {boneName=spine_03}}, {targetBone {boneName=spine_04}}]
    constraintWeights [1,1,1,1]
   }
   n10 = "Constraint - torsoPouch_2_l_zero (PPPP)"!(
    ComponentPose=@n9
    Alpha=1
    ConstraintWeights_0=0.08
    ConstraintWeights_1:float=0.24
    ConstraintWeights_2:float=0.46
    ConstraintWeights_3:float=0.22
   ) -> (Pose) {
    alphaBoolBlend {blendOption=Linear}
    boneToModify {boneName=torsoPouch_2_l_zero}
    constraintSetup [{targetBone {boneName=spine_01}}, {targetBone {boneName=spine_02}}, {targetBone {boneName=spine_03}}, {targetBone {boneName=spine_04}}]
    constraintWeights [1,1,1,1]
   }
   n11 = "Constraint - torsoPouch_1_r_zero (PPPP)"!(
    ComponentPose=@n10
    Alpha=1
    ConstraintWeights_0=0.1
    ConstraintWeights_1:float=0.24
    ConstraintWeights_2:float=0.47
    ConstraintWeights_3:float=0.18
   ) -> (Pose) {
    alphaBoolBlend {blendOption=Linear}
    boneToModify {boneName=torsoPouch_1_r_zero}
    constraintSetup [{targetBone {boneName=spine_01}}, {targetBone {boneName=spine_02}}, {targetBone {boneName=spine_03}}, {targetBone {boneName=spine_04}}]
    constraintWeights [1,1,1,1]
   }
   n12 = "Constraint - torsoPouch_2_r_zero (PPPP)"!(
    ComponentPose=@n11
    Alpha=1
    ConstraintWeights_0=0.08
    ConstraintWeights_1:float=0.24
    ConstraintWeights_2:float=0.46
    ConstraintWeights_3:float=0.22
   ) -> (Pose) {
    alphaBoolBlend {blendOption=Linear}
    boneToModify {boneName=torsoPouch_2_r_zero}
    constraintSetup [{targetBone {boneName=spine_01}}, {targetBone {boneName=spine_02}}, {targetBone {boneName=spine_03}}, {targetBone {boneName=spine_04}}]
    constraintWeights [1,1,1,1]
   }
   n13 = Anim Dynamics - Bone: torsoPouch_1_l(ComponentPose=@n12, Alpha=1, ExternalForce="0,0,0") {
    angularDampingOverride=1
    angularSpringConstant=1500
    bAngularSpring=true
    bGravityOverrideInSimSpace=true
    bOverrideAngularDamping=true
    gravityScale=0.35
    simulationSpace=RootRelative
    alphaBoolBlend {blendOption=Linear}
    alphaScaleBiasClamp {scale=0.3}
    boundBone {boneName=torsoPouch_1_l}
    physicsBodyDefinitions [{
      sphereCollisionRadius=0
      boundBone {boneName=torsoPouch_1_l}
      boxExtents=<10 20 20>
      constraintSetup {angularLimitsMax=<5 5 10>, angularLimitsMin=<-5 -5 -10>, angularTarget=<0 0 0>}
      localJointOffset=<10 0 0>
     }]
   }
   n14 = Anim Dynamics - Bone: torsoPouch_1_r(ComponentPose=@n13, Alpha=1, ExternalForce="0,0,0") {
    angularDampingOverride=1
    angularSpringConstant=1500
    bAngularSpring=true
    bGravityOverrideInSimSpace=true
    bOverrideAngularDamping=true
    gravityScale=0.35
    simulationSpace=RootRelative
    alphaBoolBlend {blendOption=Linear}
    alphaScaleBiasClamp {scale=0.3}
    boundBone {boneName=torsoPouch_1_r}
    physicsBodyDefinitions [{
      sphereCollisionRadius=0
      boundBone {boneName=torsoPouch_1_r}
      boxExtents=<10 20 20>
      constraintSetup {angularLimitsMax=<5 5 10>, angularLimitsMin=<-5 -5 -10>, angularTarget=<0 0 0>}
      localJointOffset=<10 0 0>
     }]
   }
   n15 = Anim Dynamics - Bone: torsoPouch_2_l(ComponentPose=@n14, Alpha=1, ExternalForce="0,0,0") {
    angularDampingOverride=1
    angularSpringConstant=1500
    bAngularSpring=true
    bGravityOverrideInSimSpace=true
    bOverrideAngularDamping=true
    gravityScale=0.35
    simulationSpace=RootRelative
    alphaBoolBlend {blendOption=Linear}
    alphaScaleBiasClamp {scale=0.3}
    boundBone {boneName=torsoPouch_2_l}
    physicsBodyDefinitions [{
      sphereCollisionRadius=0
      boundBone {boneName=torsoPouch_2_l}
      boxExtents=<10 20 20>
      constraintSetup {angularLimitsMax=<5 5 10>, angularLimitsMin=<-5 -5 -10>, angularTarget=<0 0 0>}
      localJointOffset=<10 0 0>
     }]
   }
   n16 = Anim Dynamics - Bone: torsoPouch_2_r(ComponentPose=@n15, Alpha=1, ExternalForce="0,0,0") {
    angularDampingOverride=1
    angularSpringConstant=1500
    bAngularSpring=true
    bGravityOverrideInSimSpace=true
    bOverrideAngularDamping=true
    gravityScale=0.35
    simulationSpace=RootRelative
    alphaBoolBlend {blendOption=Linear}
    alphaScaleBiasClamp {scale=0.3}
    boundBone {boneName=torsoPouch_2_r}
    physicsBodyDefinitions [{
      sphereCollisionRadius=0
      boundBone {boneName=torsoPouch_2_r}
      boxExtents=<10 20 20>
      constraintSetup {angularLimitsMax=<5 5 10>, angularLimitsMin=<-5 -5 -10>, angularTarget=<0 0 0>}
      localJointOffset=<10 0 0>
     }]
   }
   n17 = "Constraint - dyn_hoodbase (P)"(ComponentPose=@n16, Alpha=0, ConstraintWeights_0=1) {
    alphaBoolBlend {blendOption=Linear}
    boneToModify {boneName=dyn_hoodbase}
    constraintSetup [{targetBone {boneName=head}}]
    constraintWeights [1]
   }
   Output Pose(Result=Component To Local(ComponentPose=@n17)) {layerGroup=None}
  ]}
 EventGraph {nodes [
   TryGetPawnOwner()
   GetMetadata!(SubjectFrameHandle:"Subject Frame Handle Structure (by ref)") -> (Metadata:Subject Metadata Structure)
   Event BlueprintUpdateAnimation()
   Set Mask?(Mask?:bool=CharacterItemDefinitionHasAnyMetaTag(Tags=Make Array("[0]"=(TagName="Cosmetics.MetaTags.Style.OptionB")))) -> (Output_Get:bool)
  ]}
]
variables [
 head_rot_back {type=number}
 Mask? {type=boolean}
]
